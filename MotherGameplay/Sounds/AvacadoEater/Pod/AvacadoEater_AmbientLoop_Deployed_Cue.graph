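# Sound cue for the deployed AvacadoEater pod's ambient loop, played at low volume on the gadgets sound class with a concurrency cap.
# It mixes two layers with separate attenuations: a general looping ambience and a close-range layer.
# The close layer combines a second ambient loop with random egg and inside-egg vocals, repeated endlessly after randomized delays.
# Slight pitch modulation varies each loop and the overall mix.
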SoundCue /MotherGameplay/Sounds/AvacadoEater/Pod/AvacadoEater_AmbientLoop_Deployed_Cue
root=Modulator_0
nodes [
 {
  id=Attenuation_0
  children [Modulator_1]
  properties {attenuationSettings=/MotherGameplay/Sounds/AvacadoEater/Attenuations/AvacadoEater_AmbientLoop_Att}
 }
 {
  id=Attenuation_1
  children [Mixer_1]
  properties {attenuationSettings=/MotherGameplay/Sounds/AvacadoEater/Attenuations/AvacadoEater_AmbientLoop_Close_Att}
 }
 {id=Delay_0, children [Random_0], properties {delayMax=4, delayMin=0.1}}
 {id=Looping_0, children [Delay_0]}
 {id=Mixer_0, children [Attenuation_0, Attenuation_1], properties {inputVolume [1,1]}}
 {id=Mixer_1, children [Modulator_2,Looping_0], properties {inputVolume [1,1]}}
 {id=Modulator_0, children [Mixer_0], properties {pitchMax=1.02, pitchMin=0.98}}
 {id=Modulator_1, children [WavePlayer_4], properties {pitchMin=0.98}}
 {id=Modulator_2, children [WavePlayer_2], properties {pitchMin=0.98}}
 {
  id=Random_0
  children [
   WavePlayer_0
   WavePlayer_1
   WavePlayer_5
   WavePlayer_6
   WavePlayer_7
   WavePlayer_8
   WavePlayer_9
   WavePlayer_10
   WavePlayer_11
   WavePlayer_12
   WavePlayer_13
   WavePlayer_14
   WavePlayer_15
   WavePlayer_16
   WavePlayer_17
  ]
  properties {weights [1,1,1,1,1,1,1,1,1,1,1,1,1,1,1]}
 }
 {id=WavePlayer_0, properties {soundWaveAssetPtr=&/AvacadoEater_Egg_Vocal_10}}
 {id=WavePlayer_1, properties {soundWaveAssetPtr=&/AvacadoEater_Egg_Vocal_11}}
 {id=WavePlayer_10, properties {soundWaveAssetPtr=&/AvacadoEater_Egg_Vocal_05}}
 {id=WavePlayer_11, properties {soundWaveAssetPtr=&/AvacadoEater_Egg_Vocal_06}}
 {id=WavePlayer_12, properties {soundWaveAssetPtr=&/AvacadoEater_Egg_Vocal_07}}
 {id=WavePlayer_13, properties {soundWaveAssetPtr=&/AvacadoEater_Egg_Vocal_08}}
 {id=WavePlayer_14, properties {soundWaveAssetPtr=&/AvacadoEater_Egg_Vocal_09}}
 {id=WavePlayer_15, properties {soundWaveAssetPtr=&/AvacadoEater_InsideEgg_Vocal_01}}
 {id=WavePlayer_16, properties {soundWaveAssetPtr=&/AvacadoEater_InsideEgg_Vocal_02}}
 {id=WavePlayer_17, properties {soundWaveAssetPtr=&/AvacadoEater_InsideEgg_Vocal_03}}
 {id=WavePlayer_2, properties {bLooping=true, soundWaveAssetPtr=&/AvacadoEater_Ambient_Close_Loop}}
 {id=WavePlayer_4, properties {bLooping=true, soundWaveAssetPtr=&/AvacadoEater_Ambient_Loop_01}}
 {id=WavePlayer_5, properties {soundWaveAssetPtr=&/AvacadoEater_Egg_Vocal_12}}
 {id=WavePlayer_6, properties {soundWaveAssetPtr=&/AvacadoEater_Egg_Vocal_01}}
 {id=WavePlayer_7, properties {soundWaveAssetPtr=&/AvacadoEater_Egg_Vocal_02}}
 {id=WavePlayer_8, properties {soundWaveAssetPtr=&/AvacadoEater_Egg_Vocal_03}}
 {id=WavePlayer_9, properties {soundWaveAssetPtr=&/AvacadoEater_Egg_Vocal_04}}
]
properties {
 bOverrideConcurrency=true
 maxDistance=2848
 soundClassObject=&Fort_Gadgets
 volumeMultiplier=0.2
 concurrencyOverrides {maxCount=2}
}
