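# Sound cue for the destruction of the AvacadoEater pod, using wave assets from the Bird folder.
# A Mixer plays two layers together at equal volume: a modulated random pick of three AvacadoEater destroyed sounds, and a random pick of two Cuddle Fish explosion sounds.
# It is routed to the impacts sound class and uses bullet-impact attenuation for spatial falloff.
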SoundCue /MotherGameplay/Sounds/AvacadoEater/Pod/AvacadoEater_Destroyed_Cue
root=Mixer_0
nodes [
 {id=Mixer_0, children [Modulator_1,Random_2], properties {inputVolume [1,1]}}
 {id=Modulator_1, children [Random_0]}
 {id=Random_0, children [WavePlayer_5,WavePlayer_8,WavePlayer_9], properties {weights [1,1,1]}}
 {id=Random_2, children [WavePlayer_2,WavePlayer_3], properties {weights [1,1]}}
 {id=WavePlayer_2, properties {soundWaveAssetPtr=/MotherGameplay/Sounds/AvacadoEater/Bird/Cuddle_Fish_Explo_01}}
 {id=WavePlayer_3, properties {soundWaveAssetPtr=/MotherGameplay/Sounds/AvacadoEater/Bird/Cuddle_Fish_Explo_02}}
 {id=WavePlayer_5, properties {soundWaveAssetPtr=/MotherGameplay/Sounds/AvacadoEater/Bird/AvacadoEater_Destroyed_01}}
 {id=WavePlayer_8, properties {soundWaveAssetPtr=/MotherGameplay/Sounds/AvacadoEater/Bird/AvacadoEater_Destroyed_02}}
 {id=WavePlayer_9, properties {soundWaveAssetPtr=/MotherGameplay/Sounds/AvacadoEater/Bird/AvacadoEater_Destroyed_03}}
]
properties {
 attenuationSettings=/Game/Sounds/Attenuation/Fort_Bullet_Impact_Att
 maxDistance=5300
 soundClassObject=&Fort_Impacts
}
