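# Sound cue for the AvacadoEater pod's armed hatch sound, classed under Fort_Impacts.
# A quality-level root feeds its first branch through a slight pitch modulator into the single Hatch_Armed wave; the other branches are empty or hold modulators with no wave.
# It uses the shared AvacadoEater attenuation and overrides concurrency to cap simultaneous instances, with the debug flag left on.
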SoundCue /MotherGameplay/Sounds/AvacadoEater/Pod/AvacadoEater_Hatch_Armed_Cue
root=QualityLevel_0
nodes [
 {id=Modulator_0, children [null], properties {pitchMax=1.07, pitchMin=0.92}}
 {id=Modulator_12, children [WavePlayer_3], properties {pitchMin=0.98}}
 {id=Modulator_2, children [null], properties {pitchMax=1.07, pitchMin=0.92}}
 {id=QualityLevel_0, children [Modulator_12,null,Modulator_0,Modulator_2]}
 {id=WavePlayer_3, properties {soundWaveAssetPtr=&/AvacadoEater_Hatch_Armed}}
]
properties {
 attenuationSettings=/MotherGameplay/Sounds/AvacadoEater/Attenuations/AvacadoEater_Att
 bDebug=true
 bOverrideConcurrency=true
 maxDistance=3072
 soundClassObject=&Fort_Impacts
 volumeMultiplier=1
 concurrencyOverrides {maxCount=2}
}
