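# Sound cue for the AvacadoEater pod hatch pop, routed to the impacts sound class with shared AvacadoEater attenuation.
# A top mixer plays a single bird egg pop immediately, layered with a briefly delayed sub-mix.
# That delayed layer combines a random hatch pop variant and a random Cuddle Fish explosion, each with slight pitch and volume modulation.
# A concurrency override caps simultaneous instances.
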SoundCue /MotherGameplay/Sounds/AvacadoEater/Pod/AvacadoEater_Hatch_Pop_Cue
root=Mixer_0
nodes [
 {id=Delay_0, children [Mixer_1], properties {delayMax=0.2, delayMin=0.15}}
 {id=Mixer_0, children [Delay_0,WavePlayer_2,null], properties {inputVolume [1,1,1]}}
 {id=Mixer_1, children [Modulator_0,Modulator_1], properties {inputVolume [1,1]}}
 {id=Modulator_0, children [Random_2], properties {pitchMax=1.03, pitchMin=1, volumeMax=1, volumeMin=0.98}}
 {id=Modulator_1, children [Random_1], properties {pitchMax=1.03, pitchMin=1, volumeMax=1, volumeMin=0.98}}
 {id=Random_1, children [WavePlayer_4,WavePlayer_5], properties {weights [1,1]}}
 {id=Random_2, children [WavePlayer_8,WavePlayer_6,WavePlayer_7], properties {weights [1,1,1]}}
 {id=WavePlayer_2, properties {soundWaveAssetPtr=&/AvacadoEaterBird_Egg_Pop_01}}
 {id=WavePlayer_4, properties {soundWaveAssetPtr=/MotherGameplay/Sounds/AvacadoEater/Bird/Cuddle_Fish_Explo_01}}
 {id=WavePlayer_5, properties {soundWaveAssetPtr=/MotherGameplay/Sounds/AvacadoEater/Bird/Cuddle_Fish_Explo_02}}
 {id=WavePlayer_6, properties {soundWaveAssetPtr=&/AvacadoEater_Hatch_Pop_02}}
 {id=WavePlayer_7, properties {soundWaveAssetPtr=&/AvacadoEater_Hatch_Pop_03}}
 {id=WavePlayer_8, properties {soundWaveAssetPtr=&/AvacadoEater_Hatch_Pop_01}}
]
properties {
 attenuationSettings=/MotherGameplay/Sounds/AvacadoEater/Attenuations/AvacadoEater_Att
 bOverrideConcurrency=true
 maxDistance=3072
 soundClassObject=&Fort_Impacts
 volumeMultiplier=1.2
 attenuationOverrides {lPFFrequencyAtMax=5000, lPFRadiusMax=1000, lPFRadiusMin=200, attenuationShapeExtents {x=200}}
 concurrencyOverrides {maxCount=3}
}
