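# Sound cue named for the start of an AvacadoEater pod hatch, routed to the Fort_Impacts sound class.
# Its only node is a Quality Level root with four empty inputs, so the graph references no audio.
# It uses the shared AvacadoEater attenuation, overrides concurrency to cap simultaneous instances, boosts volume slightly, and has the debug flag enabled.
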SoundCue /MotherGameplay/Sounds/AvacadoEater/Pod/AvacadoEater_Hatch_Start_Cue
root=QualityLevel_0
nodes [{id=QualityLevel_0, children [null,null,null,null]}]
properties {
 attenuationSettings=/MotherGameplay/Sounds/AvacadoEater/Attenuations/AvacadoEater_Att
 bDebug=true
 bOverrideConcurrency=true
 maxDistance=3072
 soundClassObject=&Fort_Impacts
 volumeMultiplier=1.25
 concurrencyOverrides {maxCount=2}
}
