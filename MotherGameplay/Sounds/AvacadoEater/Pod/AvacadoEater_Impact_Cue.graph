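# SoundCue for the AvacadoEater pod impact sound, routed to the Fort impacts sound class with bullet-impact attenuation.
# A Mixer plays two layers at once, each from its own equally weighted Random node.
# One layer picks from four DogHouse AntEgg demolition impacts borrowed from shared game sounds, the other from three AvacadoEater impact waves, varying each hit.
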SoundCue /MotherGameplay/Sounds/AvacadoEater/Pod/AvacadoEater_Impact_Cue
root=Mixer_0
nodes [
 {id=Mixer_0, children [Random_0,Random_1], properties {inputVolume [1,1]}}
 {id=Random_0, children [WavePlayer_2,WavePlayer_3,WavePlayer_4,WavePlayer_5], properties {weights [1,1,1,1]}}
 {id=Random_1, children [WavePlayer_0,WavePlayer_1,WavePlayer_6], properties {weights [1,1,1]}}
 {id=WavePlayer_0, properties {soundWaveAssetPtr=&/AvacadoEater_Impact_01}}
 {id=WavePlayer_1, properties {soundWaveAssetPtr=&/AvacadoEater_Impact_02}}
 {id=WavePlayer_2, properties {soundWaveAssetPtr=/Game/Sounds/Fort_Impact_Sounds/Demolition/DogHouse/DogHouse_AntEgg_Impact_01}}
 {id=WavePlayer_3, properties {soundWaveAssetPtr=/Game/Sounds/Fort_Impact_Sounds/Demolition/DogHouse/DogHouse_AntEgg_Impact_02}}
 {id=WavePlayer_4, properties {soundWaveAssetPtr=/Game/Sounds/Fort_Impact_Sounds/Demolition/DogHouse/DogHouse_AntEgg_Impact_03}}
 {id=WavePlayer_5, properties {soundWaveAssetPtr=/Game/Sounds/Fort_Impact_Sounds/Demolition/DogHouse/DogHouse_AntEgg_Impact_04}}
 {id=WavePlayer_6, properties {soundWaveAssetPtr=&/AvacadoEater_Impact_03}}
]
properties {
 attenuationSettings=/Game/Sounds/Attenuation/Fort_Bullet_Impact_Att
 maxDistance=5300
 soundClassObject=&Fort_Impacts
 volumeMultiplier=1
}
